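; Paint brush actor for flower fields, drawn either as a plane or as spline meshes, that feeds BP_FlowerStripePainterRegion render targets.
; The construction script builds M_BrushMat instances with color, falloff and sort priority, and redraws all regions while active.
; CopySplines rebuilds the spline from another actor's spline meshes, flattened to brush height and sorted by distance from the brush.
; EnableLiveUpdate activates and shows every brush, then redraws the regions.
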(event EventBeginPlay)

(event Collision|EventActorBeginOverlap
  (params (OtherActor :type Actor%20Object%20Reference)))

(event EventTick
  (params (DeltaSeconds :type Float%20%28single-precision%29)))

(event Custom|EnableLiveUpdate
  (Utilities|FlowControl|Sequence
    (:then_0
      (bind _outactors (Actor|GetAllActorsOfClass "/Hermes/Techart/FlowerFields/Blueprint/BP_FlowerStripePaintBrush.BP_FlowerStripePaintBrush_C"))
      (for _array_element _outactors
        (bind _output_get (Variables|Default|SetisActive true _array_element))
        (Rendering|SetVisibility (Variables|Default|GetPlane _array_element) _output_get false)
        (CallFunction|ShowallSplines _array_element)))
    (:then_1
      (bind _outactors_2 (Actor|GetAllActorsOfClass "/Hermes/Techart/FlowerFields/Blueprint/BP_FlowerStripePainterRegion.BP_FlowerStripePainterRegion_C"))
      (for _array_element_2 _outactors_2
        (Class|BPFlowerStripePainterRegion|DrawAllBrushestoRenderTarget _array_element_2)))))

(fn CopySplines
  (bind _returnvalue (Transformation|GetActorLocation))
  (bind _returnvalue_2 (Transformation|GetActorLocation))
  (bind _spline (Variables|Default|GetSpline))
  (bind _spline_2 (Variables|Default|GetSpline))
  (bind _spline_3 (Variables|Default|GetSpline))
  (bind _scalepoints (LocalVariables|GetScalePoints))
  (bind _distance (LocalVariables|GetDistance))
  (bind _copylandscapespline (Variables|Default|GetCopyLandScapeSpline)
    (:then
      (Spline|SetOverrideConstructionScript _spline_3 false)
      (Variables|/Script/Engine.SplineComponent|SetbInputSplinePointsToConstructionScript true _spline_3)
      (Utilities|FlowControl|Sequence
        (:then_0
          (for _index (range (+ (Spline|GetNumberOfSplinePoints _spline) 1))
            (Spline|RemoveSplinePoint _spline _index true)))
        (:then_1
          (for (_array_element _array_index) (Actor|GetComponentsByClass _copylandscapespline "/Script/Engine.SplineMeshComponent")
            (bind _asspline_mesh_component (Utilities|Casting|CastToSplineMeshComponent _array_element))
            (Utilities|Array|Add (LocalVariables|GetSplinePointsLocal) (+ (* (+ (Math|Vector|RotateVector (SplineMesh|GetStartPosition _asspline_mesh_component) (.rotation (Transformation|GetWorldTransform _asspline_mesh_component))) (.location (Transformation|GetWorldTransform _asspline_mesh_component))) "1.000000,1.000000,0.000000") (* _returnvalue_2 "0.000000,0.000000,1.000000")))
            (Utilities|Array|Add (LocalVariables|GetSplinePointsTangents) (Math|Vector|RotateVector (SplineMesh|GetStartTangent _asspline_mesh_component) (.rotation (Transformation|GetWorldTransform _asspline_mesh_component))))
            (Utilities|Array|Add (LocalVariables|GetDistance) (Math|Vector|Distance(Vector) (* _returnvalue_2 "1.000000,1.000000,0.000000") (* (+ (Math|Vector|RotateVector (SplineMesh|GetStartPosition _asspline_mesh_component) (.rotation (Transformation|GetWorldTransform _asspline_mesh_component))) (.location (Transformation|GetWorldTransform _asspline_mesh_component))) "1.000000,1.000000,0.000000")))
            (Utilities|Array|Add (LocalVariables|GetScalePoints) (SplineMesh|GetStartScale _asspline_mesh_component))
            (if (== (- (Utilities|Array|Length (Actor|GetComponentsByClass _copylandscapespline "/Script/Engine.SplineMeshComponent")) 1) _array_index)
              (Utilities|Array|Add (LocalVariables|GetSplinePointsLocal) (+ (* (+ (Math|Vector|RotateVector (SplineMesh|GetEndPosition _asspline_mesh_component) (.rotation (Transformation|GetWorldTransform _asspline_mesh_component))) (.location (Transformation|GetWorldTransform _asspline_mesh_component))) "1.000000,1.000000,0.000000") (* _returnvalue "0.000000,0.000000,1.000000")))
              (Utilities|Array|Add (LocalVariables|GetSplinePointsTangents) (Math|Vector|RotateVector (SplineMesh|GetEndTangent _asspline_mesh_component) (.rotation (Transformation|GetWorldTransform _asspline_mesh_component))))
              (Utilities|Array|Add (LocalVariables|GetDistance) (Math|Vector|Distance(Vector) (* _returnvalue "1.000000,1.000000,0.000000") (* (+ (Math|Vector|RotateVector (SplineMesh|GetEndPosition _asspline_mesh_component) (.rotation (Transformation|GetWorldTransform _asspline_mesh_component))) (.location (Transformation|GetWorldTransform _asspline_mesh_component))) "1.000000,1.000000,0.000000")))
              (Utilities|Array|Add (LocalVariables|GetScalePoints) (SplineMesh|GetEndScale _asspline_mesh_component)))))
        (:then_2
          (LocalVariables|SetDistance_Sorted (LocalVariables|GetDistance))
          (Utilities|Array|Sort|SortFloatArray (LocalVariables|GetDistance_Sorted))
          (for _array_element_2 (LocalVariables|GetDistance_Sorted)
            (Utilities|Array|Add (LocalVariables|GetSplinePointsLocal_Soreted) (Utilities|Array|Get(acopy) (LocalVariables|GetSplinePointsLocal) (Utilities|Array|FindItem (LocalVariables|GetDistance) _array_element_2)))))
        (:then_3
          (Spline|SetSplinePoints _spline_3 (LocalVariables|GetSplinePointsLocal_Soreted) "World" true)
          (for (_array_element_3 _array_index_2) (LocalVariables|GetDistance_Sorted)
            (Spline|SetTangentatSplinePoint _spline_2 _array_index_2 (Utilities|Array|Get(acopy) (LocalVariables|GetSplinePointsTangents) (Utilities|Array|FindItem _distance _array_element_3)) "World" true)
            (Spline|SetScaleatSplinePoint _spline_2 _array_index_2 (Math|Vector|MakeVector (Math|Vector2D|BreakVector2D (Utilities|Array|Get(acopy) _scalepoints (Utilities|Array|FindItem _distance _array_element_3))) (.Y (Math|Vector2D|BreakVector2D (Utilities|Array|Get(acopy) _scalepoints (Utilities|Array|FindItem _distance _array_element_3)))) 1.0) true)
            (Development|PrintString (Utilities|String|ToString(Vector) (Math|Vector|MakeVector (Math|Vector2D|BreakVector2D (Utilities|Array|Get(acopy) _scalepoints (Utilities|Array|FindItem _distance _array_element_3))) (.Y (Math|Vector2D|BreakVector2D (Utilities|Array|Get(acopy) _scalepoints (Utilities|Array|FindItem _distance _array_element_3)))) 1.0)))))
        (:then_4
          (Spline|SetOverrideConstructionScript _spline_3 true)
          (Variables|/Script/Engine.SplineComponent|SetbInputSplinePointsToConstructionScript true _spline_3)
          (Variables|Default|SetCopyLandScapeSpline 0))))
    (:else)))

(fn ShowallSplines
  (for _array_element (Variables|Default|GetSplineMeshes)
    (Rendering|SetVisibility _array_element true false)))

(fn ConstructionScript
  (bind _spline (Variables|Default|GetSpline))
  (bind _usespline (Variables|Default|GetUseSpline))
  (bind _spline_2 (Variables|Default|GetSpline))
  (bind _plane (Variables|Default|GetPlane))
  (Utilities|FlowControl|Sequence
    (:then_0
      (CallFunction|CopySplines)
      (Utilities|FlowControl|Sequence
        (:then_0
          (Rendering|SetVisibility (Variables|Default|GetSpline) _usespline false)
          (if _usespline
            (Transformation|SetActorScale3D "1.000000,1.000000,1.000000")
            (Transformation|SetRelativeRotation _spline_2 "0, 0, 0" false false)
            (Transformation|SetRelativeLocation _spline_2 "0, 0, 0" false false)
            (Utilities|FlowControl|Sequence
              (:then_0
                (for _index (range (+ (- (Spline|GetNumberOfSplinePoints _spline) 2) 1))
                  (bind _returnvalue (AddComponent|NODE_AddSplineMeshComponent-0))
                  (Rendering|Material|SetMaterial _returnvalue 0 "/Engine/BasicShapes/BasicShapeMaterial.BasicShapeMaterial")
                  (SplineMesh|SetStartandEnd _returnvalue (Spline|GetLocationandTangentatSplinePoint _spline _index "Local") (.Tangent (Spline|GetLocationandTangentatSplinePoint _spline _index "Local")) (Spline|GetLocationandTangentatSplinePoint _spline (+ _index 1) "Local") (.Tangent (Spline|GetLocationandTangentatSplinePoint _spline (+ _index 1) "Local")) true :!class /Script/Engine.SplineMeshComponent)
                  (SplineMesh|SetStartScale _returnvalue (Math|Vector2D|MakeVector2D (.y (Spline|GetScaleatSplinePoint _spline _index)) 1.0) true :!class /Script/Engine.SplineMeshComponent)
                  (SplineMesh|SetEndScale _returnvalue (Math|Vector2D|MakeVector2D (.y (Spline|GetScaleatSplinePoint _spline (+ _index 1))) 1.0) true :!class /Script/Engine.SplineMeshComponent)
                  (Utilities|Array|Add (Variables|Default|GetSplineMeshes) _returnvalue)))
              (:then_1
                (for _array_element (Variables|Default|GetSplineMeshes)
                  (Rendering|SetVisibility _array_element (Variables|Default|GetisActive) false)
                  (Rendering|SetTranslucentSortPriority _array_element (Variables|Default|GetPriority))
                  (bind _returnvalue_2 (Rendering|Material|CreateDynamicMaterialInstance _array_element 0 "/Hermes/Techart/FlowerFields/Blueprint/M_BrushMat.M_BrushMat" "None" :!class /Script/Engine.PrimitiveComponent))
                  (bind _output_get (Variables|Default|SetMID _returnvalue_2))
                  (Rendering|Material|SetVectorParameterValue _output_get "Color" (Variables|Default|GetColor) :!class /Script/Engine.MaterialInstanceDynamic)
                  (Rendering|Material|SetScalarParameterValue _output_get "FalloffRadial" (Variables|Default|GetFalloffRadial) :!class /Script/Engine.MaterialInstanceDynamic)
                  (Rendering|Material|SetScalarParameterValue _output_get "FalloffRectX" (Variables|Default|GetFalloffRectangle%20X) :!class /Script/Engine.MaterialInstanceDynamic)
                  (Rendering|Material|SetScalarParameterValue _output_get "FalloffRectY" 9999999.0 :!class /Script/Engine.MaterialInstanceDynamic)
                  (Rendering|Material|SetScalarParameterValue _output_get "SphereRectangle" 1.0 :!class /Script/Engine.MaterialInstanceDynamic))))))
        (:then_3
          (Utilities|FlowControl|Sequence
            (:then_0
              (bind _outactors (Actor|GetAllActorsOfClass "/Hermes/Techart/FlowerFields/Blueprint/BP_FlowerStripePaintBrush.BP_FlowerStripePaintBrush_C"))
              (bind _array_element_2 (Utilities|Array|ForEachLoopwithBreak _outactors)
                (:LoopBody
                  (if (== _array_element_2 self)
                    _
                    (elif (Variables|Default|GetisActive _array_element_2)
                      (Variables|Default|SetisActive true))))))
            (:then_1
              (Rendering|SetTranslucentSortPriority _plane (Variables|Default|GetPriority))
              (bind _returnvalue_3 (Rendering|Material|CreateDynamicMaterialInstance _plane 0 "/Hermes/Techart/FlowerFields/Blueprint/M_BrushMat.M_BrushMat" "None" :!class /Script/Engine.PrimitiveComponent))
              (bind _output_get_2 (Variables|Default|SetMID _returnvalue_3))
              (Rendering|Material|SetVectorParameterValue _output_get_2 "Color" (Variables|Default|GetColor) :!class /Script/Engine.MaterialInstanceDynamic)
              (Rendering|Material|SetScalarParameterValue _output_get_2 "FalloffRadial" (Variables|Default|GetFalloffRadial) :!class /Script/Engine.MaterialInstanceDynamic)
              (Rendering|Material|SetScalarParameterValue _output_get_2 "FalloffRectX" (Variables|Default|GetFalloffRectangle%20X) :!class /Script/Engine.MaterialInstanceDynamic)
              (Rendering|Material|SetScalarParameterValue _output_get_2 "FalloffRectY" (Variables|Default|GetFalloffRectangle%20Y) :!class /Script/Engine.MaterialInstanceDynamic)
              (Rendering|Material|SetScalarParameterValue _output_get_2 "SphereRectangle" (Variables|Default|GetBlendShape_Sphere_Rectangle) :!class /Script/Engine.MaterialInstanceDynamic))
            (:then_2
              (Rendering|SetVisibility (Variables|Default|GetPlane) (Variables|Default|GetisActive) false))))
        (:then_4
          (if (Variables|Default|GetUseSpline)
            (Components|DestroyComponent (Variables|Default|GetPlane))))
        (:then_5
          (if (Variables|Default|GetisActive)
            (bind _outactors_2 (Actor|GetAllActorsOfClass "/Hermes/Techart/FlowerFields/Blueprint/BP_FlowerStripePainterRegion.BP_FlowerStripePainterRegion_C"))
            (for _array_element_3 _outactors_2
              (Class|BPFlowerStripePainterRegion|DrawAllBrushestoRenderTarget _array_element_3))))))))
