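; Flower-field region painter that bakes brush-driven color and grass mask textures over a square area, using editor-only render target conversion.
; DrawAllBrushesToRenderTarget composites BP_FlowerStripePaintBrush actors by ascending priority, scene-capturing spline brushes, adds noise and hue overlays, then saves FlowerColorProjection.
; CaptureGrassMask temporarily clones flower-field landscape spline meshes with a mask material and captures FlowerGrassMask.
; Construction syncs bounds, captures and MPC; GenerateHLODMesh spawns a geoscript child actor.
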(event EventBeginPlay)

(event Collision|EventActorBeginOverlap
  (params (OtherActor :type Actor%20Object%20Reference)))

(event EventTick
  (params (DeltaSeconds :type Float%20%28single-precision%29)))

(event Custom|UpdateTexture
  (CallFunction|DrawAllBrushestoRenderTarget))

(event Custom|GenerateHLODMesh
  (bind _returnvalue (AddComponent|NODE_AddChildActorComponent-0 :RelativeTransform (Math|Transform|MakeTransform "0, 0, 0")))
  (Variables|Default|SetNewVar _returnvalue)
  (bind _asbp_geoscript_flower_fields (Utilities|Casting|CastToBP_GeoscriptFlowerFields (Variables|/Script/Engine.ChildActorComponent|GetChildActor _returnvalue)))
  (Variables|/Hermes/Techart/FlowerFields/Blueprint/BP_GeoscriptFlowerFields.BP_GeoscriptFlowerFields_C|SetOwningActor self _asbp_geoscript_flower_fields))

(event Custom|DestroyGeoscript
  (Components|DestroyComponent (stale Custom|GenerateHLODMesh _returnvalue)))

(fn CaptureGrassMask
  (bind _scenecapturecomponent2d (Variables|Default|GetSceneCaptureComponent2D))
  (bind _sizesquared (Variables|Default|GetSizeSquared))
  (Utilities|Array|Clear (Variables|Default|GetActorList))
  (bind _outactors (Actor|GetAllActorsOfClass "/Script/Landscape.LandscapeSplineActor"))
  (for _array_element _outactors
    (for _array_element_2 (Actor|GetComponentsByClass _array_element "/Script/Engine.SplineMeshComponent")
      (Utilities|IsValid _array_element_2
        (:Is%20Valid
          (if (Math|Boolean|ORBoolean (== (Variables|/Script/Engine.StaticMeshComponent|GetStaticMesh _array_element_2) "/Hermes/Techart/FlowerFields/Meshes/SM_FlowerFields_Straight.SM_FlowerFields_Straight") (== (Variables|/Script/Engine.StaticMeshComponent|GetStaticMesh _array_element_2) "/Hermes/Techart/FlowerFields/Meshes/SM_FlowerFields_End.SM_FlowerFields_End") (== (Variables|/Script/Engine.StaticMeshComponent|GetStaticMesh _array_element_2) "/Hermes/Techart/FlowerFields/Meshes/SM_FlowerFields_Curve_A_A.SM_FlowerFields_Curve_A_A") (== (Variables|/Script/Engine.StaticMeshComponent|GetStaticMesh _array_element_2) "/Hermes/Techart/FlowerFields/Meshes/SM_FlowerFields_Curve_A_B.SM_FlowerFields_Curve_A_B") (== (Variables|/Script/Engine.StaticMeshComponent|GetStaticMesh _array_element_2) "/Hermes/Techart/FlowerFields/Meshes/SM_FlowerFields_Curve_A_C.SM_FlowerFields_Curve_A_C") (== (Variables|/Script/Engine.StaticMeshComponent|GetStaticMesh _array_element_2) "/Hermes/Techart/FlowerFields/Meshes/SM_FlowerFields_Curve_A_D.SM_FlowerFields_Curve_A_D"))
            (Rendering|Material|SetDefaultCustomPrimitiveDataVector3 _array_element_2 0 (- (Transformation|GetActorLocation) (/ _sizesquared 2.0)))
            (Rendering|Material|SetDefaultCustomPrimitiveDataFloat _array_element_2 4 _sizesquared)
            (bind _returnvalue (AddComponent|NODE_AddSplineMeshComponent-0 :RelativeTransform (Math|Transform|MakeTransform (+ (- (Transformation|GetWorldLocation _array_element_2) (Transformation|GetActorLocation)) "0.000000,0.000000,100.000000") (Transformation|GetWorldRotation _array_element_2))))
            (SplineMesh|SetForwardAxis _returnvalue (SplineMesh|GetForwardAxis _array_element_2) true)
            (SplineMesh|SetSplineUpDir _returnvalue (Transformation|GetUpVector _array_element_2) true)
            (SplineMesh|SetEndRollRadians _returnvalue (SplineMesh|GetEndRoll _array_element_2) true)
            (SplineMesh|SetStartRollRadians _returnvalue (SplineMesh|GetStartRoll _array_element_2) true)
            (Components|StaticMesh|SetStaticMesh _returnvalue (Variables|/Script/Engine.StaticMeshComponent|GetStaticMesh _array_element_2))
            (SplineMesh|SetStartandEnd _returnvalue (SplineMesh|GetStartPosition _array_element_2) (SplineMesh|GetStartTangent _array_element_2) (SplineMesh|GetEndPosition _array_element_2) (SplineMesh|GetEndTangent _array_element_2) true :!class /Script/Engine.SplineMeshComponent)
            (SplineMesh|SetStartScale _returnvalue (SplineMesh|GetStartScale _array_element_2) true :!class /Script/Engine.SplineMeshComponent)
            (SplineMesh|SetEndScale _returnvalue (SplineMesh|GetEndScale _array_element_2) true :!class /Script/Engine.SplineMeshComponent)
            (Rendering|Material|SetMaterial _returnvalue 0 "/Hermes/Techart/FlowerFields/Blueprint/M_SplineMaskTempMat.M_SplineMaskTempMat")
            (Utilities|Array|Add (Variables|Default|GetSpline%20List) _returnvalue))))))
  (Development|PrintString (Utilities|String|ToString(Integer) (Utilities|Array|Length _outactors)))
  (Variables|/Script/Engine.SceneCaptureComponent|SetShowOnlyActors (Utilities|Array|MakeArray self) _scenecapturecomponent2d)
  (Rendering|SceneCapture|CaptureScene _scenecapturecomponent2d :!class /Script/Engine.SceneCaptureComponent2D)
  (EditorScripting|Rendering|ConvertRenderTargettoTexture2DEditorOnly "/Hermes/Techart/FlowerFields/Blueprint/RT_FlowerStrips_mask.RT_FlowerStrips_mask" "/Hermes/Techart/FlowerFields/Blueprint/FlowerGrassMask.FlowerGrassMask")
  (for _array_element_3 (Variables|Default|GetSpline%20List)
    (Components|DestroyComponent _array_element_3)))

(fn DrawAllBrushesToRenderTarget
  (bind _arrayprioritylist (Variables|Default|GetArrayPrioritylist))
  (bind (_indexofminvalue _minvalue) (Math|Integer|MinOfIntArray _arrayprioritylist))
  (bind _scenecapturecomponent2dsplines (Variables|Default|GetSceneCaptureComponent2DSplines))
  (bind _sizesquared (Variables|Default|GetSizeSquared))
  (Utilities|FlowControl|Sequence
    (:then_0
      (Utilities|Array|Clear (Variables|Default|GetArrayPrioritylist))
      (Utilities|Array|Clear (Variables|Default|GetArrayPrioritylist_Sorted))
      (Rendering|ClearRenderTarget2D (Variables|Default|GetRender%20Target) (Variables|Default|GetDefault%20Color))
      (bind _outactors (Actor|GetAllActorsOfClass "/Hermes/Techart/FlowerFields/Blueprint/BP_FlowerStripePaintBrush.BP_FlowerStripePaintBrush_C"))
      (Utilities|FlowControl|Sequence
        (:then_0
          (for _array_element _outactors
            (Utilities|Array|AddUnique _arrayprioritylist (Variables|/Hermes/Techart/FlowerFields/Blueprint/BP_FlowerStripePaintBrush.BP_FlowerStripePaintBrush_C|GetPriority _array_element)))
          (bind _output_get (Variables|Default|Setnumpris (- (Utilities|Array|Length _arrayprioritylist) 1)))
          (for _ (range (+ _output_get 1))
            (Utilities|Array|Add (Variables|Default|GetArrayPrioritylist_Sorted) _minvalue)
            (Utilities|Array|RemoveIndex _arrayprioritylist _indexofminvalue)))
        (:then_1
          (Variables|/Script/Engine.SceneCaptureComponent2D|SetOrthoWidth _sizesquared (Variables|Default|GetSceneCaptureComponent2D))
          (Variables|/Script/Engine.SceneCaptureComponent2D|SetOrthoWidth _sizesquared (Variables|Default|GetSceneCaptureComponent2DSplines)))
        (:then_2
          (for _array_element_2 (Variables|Default|GetArrayPrioritylist_Sorted)
            (for _array_element_3 _outactors
              (if (Variables|/Hermes/Techart/FlowerFields/Blueprint/BP_FlowerStripePaintBrush.BP_FlowerStripePaintBrush_C|GetUseSpline _array_element_3)
                (Utilities|FlowControl|Sequence
                  (:then_0
                    (Variables|/Script/Engine.SceneCaptureComponent|SetShowFlagSettings (Utilities|Array|MakeArray (Utilities|Struct|MakeEngineShowFlagsSetting "Game")) (Variables|Default|GetSceneCaptureComponent2DSplines)))
                  (:then_1
                    (Rendering|ClearRenderTarget2D (Variables|/Script/Engine.SceneCaptureComponent2D|GetTextureTarget _scenecapturecomponent2dsplines))
                    (Rendering|SetActorHiddenInGame _array_element_3 false)
                    (Variables|/Script/Engine.SceneCaptureComponent|SetShowOnlyActors (Utilities|Array|MakeArray _array_element_3) _scenecapturecomponent2dsplines)
                    (Rendering|SceneCapture|CaptureScene _scenecapturecomponent2dsplines :!class /Script/Engine.SceneCaptureComponent2D)
                    (Rendering|SetActorHiddenInGame _array_element_3 true)
                    (bind (_canvas _size) (Rendering|BeginDrawCanvastoRenderTarget (Variables|Default|GetRender%20Target)))
                    (Canvas|DrawMaterial :self _canvas :RenderMaterial "/Hermes/Techart/FlowerFields/Blueprint/M_SplineBrushMat.M_SplineBrushMat" :ScreenSize _size :CoordinatePosition "(X=0.000000,Y=0.000000)" :CoordinateSize "(X=1.000000,Y=1.000000)" :Rotation 0.0 :PivotPoint "(X=0.500000,Y=0.500000)")))
                (elif (== (Variables|/Hermes/Techart/FlowerFields/Blueprint/BP_FlowerStripePaintBrush.BP_FlowerStripePaintBrush_C|GetPriority _array_element_3) _array_element_2)
                  (bind (_canvas_2 _size_2) (Rendering|BeginDrawCanvastoRenderTarget (Variables|Default|GetRender%20Target)))
                  (Canvas|DrawMaterial _canvas_2 (Variables|/Hermes/Techart/FlowerFields/Blueprint/BP_FlowerStripePaintBrush.BP_FlowerStripePaintBrush_C|GetMID _array_element_3) (- (* _size_2 (Math|Vector2D|MakeVector2D (.x (+ (/ (- (Transformation|GetActorLocation _array_element_3) (Transformation|GetActorLocation)) (Variables|Default|GetSizeSquared)) "0.500000,0.500000,0.500000")) (.y (+ (/ (- (Transformation|GetActorLocation _array_element_3) (Transformation|GetActorLocation)) (Variables|Default|GetSizeSquared)) "0.500000,0.500000,0.500000")))) (/ (* _size_2 (Math|Vector2D|MakeVector2D (* (.x (Transformation|GetActorScale3D _array_element_3)) (/ 1e+03 (Variables|Default|GetSizeSquared))) (* (.y (Transformation|GetActorScale3D _array_element_3)) (/ 1e+03 (Variables|Default|GetSizeSquared))))) "(X=2.000000,Y=2.000000)")) (* _size_2 (Math|Vector2D|MakeVector2D (* (.x (Transformation|GetActorScale3D _array_element_3)) (/ 1e+03 (Variables|Default|GetSizeSquared))) (* (.y (Transformation|GetActorScale3D _array_element_3)) (/ 1e+03 (Variables|Default|GetSizeSquared))))) "(X=0.000000,Y=0.000000)" "(X=1.000000,Y=1.000000)" (.yaw (Transformation|GetActorRotation _array_element_3)) "(X=0.500000,Y=0.500000)"))))
            (Utilities|FlowControl|Sequence
              (:then_0
                (bind (_canvas_3 _size_3) (Rendering|BeginDrawCanvastoRenderTarget (Variables|Default|GetRender%20Target)))
                (Canvas|DrawMaterial :self _canvas_3 :RenderMaterial "/Hermes/Techart/FlowerFields/Blueprint/M_BrushOverlayNoise.M_BrushOverlayNoise" :ScreenSize _size_3 :CoordinatePosition "(X=0.000000,Y=0.000000)" :CoordinateSize "(X=1.000000,Y=1.000000)" :Rotation 0.0 :PivotPoint "(X=0.500000,Y=0.500000)"))
              (:then_1
                (bind (_canvas_4 _size_4) (Rendering|BeginDrawCanvastoRenderTarget (Variables|Default|GetRender%20Target)))
                (Canvas|DrawMaterial :self _canvas_4 :RenderMaterial "/Hermes/Techart/FlowerFields/Blueprint/M_BrushHueVariation.M_BrushHueVariation" :ScreenSize _size_4 :CoordinatePosition "(X=0.000000,Y=0.000000)" :CoordinateSize "(X=1.000000,Y=1.000000)" :Rotation 0.0 :PivotPoint "(X=0.500000,Y=0.500000)")))))))
    (:then_1
      (CallFunction|MakeStatic))))

(fn MakeStatic
  (EditorScripting|Rendering|ConvertRenderTargettoTexture2DEditorOnly "/Hermes/Techart/FlowerFields/Blueprint/RT_FlowerStrips.RT_FlowerStrips" "/Hermes/Techart/FlowerFields/Blueprint/FlowerColorProjection.FlowerColorProjection"))

(fn ConstructionScript
  (bind _sizesquared (Variables|Default|GetSizeSquared))
  (bind _returnvalue (Math|Conversions|ToVector(Float) _sizesquared))
  (bind _returnvalue_2 (/ _sizesquared 2.0))
  (bind _returnvalue_3 (Transformation|GetActorLocation))
  (bind _returnvalue_4 (- _returnvalue_3 _returnvalue_2))
  (bind _returnvalue_5 (Math|Conversions|ToLinearColor(Vector) _returnvalue_4))
  (bind _returnvalue_6 (Math|Conversions|ToLinearColor(Vector) _returnvalue))
  (bind _sizesquared_2 (Variables|Default|GetSizeSquared))
  (Utilities|FlowControl|Sequence
    (:then_0
      (Components|Box|SetBoxExtent (Variables|Default|GetBox) (Math|Conversions|ToVector(Float) (/ (Variables|Default|GetSizeSquared) 2.0)) true)
      (Rendering|Material|SetVectorParameterValue "/Hermes/Techart/FlowerFields/Blueprint/MPC_FlowerFields.MPC_FlowerFields" "FlowerField_Origin" _returnvalue_5 :!class /Script/Engine.KismetMaterialLibrary)
      (Rendering|Material|SetVectorParameterValue "/Hermes/Techart/FlowerFields/Blueprint/MPC_FlowerFields.MPC_FlowerFields" "FlowerField_Origin" _returnvalue_6 :!class /Script/Engine.KismetMaterialLibrary)
      (Variables|Default|SetMPC_OriginLocation _returnvalue_5)
      (Variables|Default|SetMPC_Size _returnvalue_6)
      (Variables|/Script/Engine.SceneCaptureComponent2D|SetOrthoWidth _sizesquared_2 (Variables|Default|GetSceneCaptureComponent2D))
      (Variables|/Script/Engine.SceneCaptureComponent2D|SetOrthoWidth _sizesquared_2 (Variables|Default|GetSceneCaptureComponent2DSplines)))))
