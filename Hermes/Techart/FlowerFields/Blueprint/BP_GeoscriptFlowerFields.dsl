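; Geometry Script generated-mesh actor that merges flower-field landscape spline meshes into one static mesh.
; On rebuild, it copies matching straight, end and curve pieces from all LandscapeSplineActors, flipping normals on mirrored ones and appending them actor-relative with a Z lift.
; It then applies planar simplification, edge welding and normal recomputation.
; Bake writes the configured static mesh, then calls DestroyGeoscript on the owning BP_FlowerStripePainterRegion.
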(event EventBeginPlay)

(event Collision|EventActorBeginOverlap
  (params (OtherActor :type Actor%20Object%20Reference)))

(event EventTick
  (params (DeltaSeconds :type Float%20%28single-precision%29)))

(event Events|EventOnRebuildGeneratedMesh
  (params (TargetMesh :type Dynamic%20Mesh%20Object%20Reference))
  (Variables|Default|SetTarget%20Mesh TargetMesh)
  (Utilities|FlowControl|Sequence
    (:then_0
      (bind _outactors (Actor|GetAllActorsOfClass "/Script/Landscape.LandscapeSplineActor"))
      (for _array_element _outactors
        (for _array_element_2 (Actor|GetComponentsByClass _array_element "/Script/Engine.SplineMeshComponent")
          (Utilities|IsValid _array_element_2
            (:Is%20Valid
              (if (Math|Boolean|ORBoolean (== (Variables|/Script/Engine.StaticMeshComponent|GetStaticMesh _array_element_2) "/Hermes/Techart/FlowerFields/Meshes/SM_FlowerFields_Straight.SM_FlowerFields_Straight") (== (Variables|/Script/Engine.StaticMeshComponent|GetStaticMesh _array_element_2) "/Hermes/Techart/FlowerFields/Meshes/SM_FlowerFields_End.SM_FlowerFields_End") (== (Variables|/Script/Engine.StaticMeshComponent|GetStaticMesh _array_element_2) "/Hermes/Techart/FlowerFields/Meshes/SM_FlowerFields_Curve_A_A.SM_FlowerFields_Curve_A_A") (== (Variables|/Script/Engine.StaticMeshComponent|GetStaticMesh _array_element_2) "/Hermes/Techart/FlowerFields/Meshes/SM_FlowerFields_Curve_A_B.SM_FlowerFields_Curve_A_B") (== (Variables|/Script/Engine.StaticMeshComponent|GetStaticMesh _array_element_2) "/Hermes/Techart/FlowerFields/Meshes/SM_FlowerFields_Curve_A_C.SM_FlowerFields_Curve_A_C") (== (Variables|/Script/Engine.StaticMeshComponent|GetStaticMesh _array_element_2) "/Hermes/Techart/FlowerFields/Meshes/SM_FlowerFields_Curve_A_D.SM_FlowerFields_Curve_A_D"))
                (bind _returnvalue (DynamicMeshActor|AllocateComputeMesh))
                (bind _output_get (Variables|Default|SetTempMesh _returnvalue))
                (bind (_localtoworld _returnvalue_2) (GeometryScript|Scene|CopyMeshfromComponent _array_element_2 _output_get (Utilities|Struct|MakeGeometryScriptCopyMeshfromComponentOptions true))
                  (:Success
                    (Development|PrintString (Utilities|String|ToString(Vector2d) (SplineMesh|GetStartScale _array_element_2)))
                    (if (xor (< (Math|Vector2D|BreakVector2D (SplineMesh|GetStartScale _array_element_2)) 0) (< (.Y (Math|Vector2D|BreakVector2D (SplineMesh|GetStartScale _array_element_2))) 0))
                      (GeometryScript|Normals|FlipNormals _returnvalue_2))
                    (GeometryScript|Transforms|TransformMesh _returnvalue_2 (Math|Transform|MakeTransform (+ (- (.location _localtoworld) (.location (Transformation|GetActorTransform))) "0.000000,0.000000,32.000000") (.rotation _localtoworld) (.scale _localtoworld)))
                    (GeometryScript|Normals|RecomputeNormals (Variables|Default|GetTempMesh))
                    (GeometryScript|MeshEdits|AppendMesh :TargetMesh (Variables|Default|GetTarget%20Mesh) :AppendMesh (Variables|Default|GetTempMesh) :bDeferChangeNotifications false :AppendOptions "(CombineMode=EnableAllMatching)")))))))))
    (:then_1
      (GeometryScript|Repair|WeldMeshEdges 0 (Utilities|Struct|MakeGeometryScriptWeldEdgesOptions 5e+03 false))
      (bind _returnvalue_3 (GeometryScript|Simplification|ApplySimplifytoPlanar (Variables|Default|GetTarget%20Mesh) (Utilities|Struct|MakeGeometryScriptPlanarSimplifyOptions 1.2)))
      (bind _returnvalue_4 (GeometryScript|Repair|WeldMeshEdges _returnvalue_3 (Utilities|Struct|MakeGeometryScriptWeldEdgesOptions 5e+03)))
      (GeometryScript|Normals|RecomputeNormals _returnvalue_4))
    (:then_2
      (CallFunction|Bake))))

(event Custom|SaveStaticMesh
  (CallFunction|Bake))

(fn Bake
  (GeometryScript|StaticMesh|CopyMeshtoStaticMesh (Variables|Default|GetTarget%20Mesh) (Variables|Default|GetStaticMeshToSave) (Utilities|Struct|MakeGeometryScriptCopyMeshtoAssetOptions false true) (Utilities|Struct|MakeGeometryScriptMeshWriteLOD false)
    (:Success
      (bind _owningactor (Variables|Default|GetOwningActor)
        (:then
          (bind _asbp_flower_stripe_painter_region (Utilities|Casting|CastToBP_FlowerStripePainterRegion _owningactor))
          (Class|BPFlowerStripePainterRegion|DestroyGeoscript _asbp_flower_stripe_painter_region))
        (:else)))))

(fn ConstructionScript)
