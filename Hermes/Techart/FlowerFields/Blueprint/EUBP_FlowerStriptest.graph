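# EUBP_FlowerStriptest is an editor utility blueprint that appears to be an unfinished test of flower-strip texture tooling for FlowerFields.
# Its Construction Script attempts to write a vector to a property named "1" on MPC_FlowerFields through SetEditorProperty.
# The EventGraph holds unconnected nodes for converting RT_FlowerStrips to a Texture2D, creating a texture asset and renaming an asset, so none of that runs.
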EditorUtilityBlueprint /Hermes/Techart/FlowerFields/Blueprint/EUBP_FlowerStriptest
graphs [
 EventGraph {nodes [
   CreateNewTexture2DAsset!(
    execute
    FromTexture:Texture 2D Object Reference
    AssetPathAndName:String
    Options:Geometry Script Create New Texture 2DAsset Options Structure
   ) -> (Failure:Exec, Success:Exec, ReturnValue:Texture 2D Object Reference)
   RenameAsset!(execute, SourceAssetPath:String, DestinationAssetPath:String) -> (then, ReturnValue:bool)
   ConvertRenderTargetToTexture2DEditorOnly(RenderTarget=&/RT_FlowerStrips)
   Event BeginPlay()
   Event ActorBeginOverlap()
   Event Tick!() -> (OutputDelegate, then, DeltaSeconds:float)
  ]}
 UserConstructionScript {nodes [
   Construction Script()
   > SetEditorProperty(Object=&/MPC_FlowerFields, PropertyName=1, PropertyValue:Vector=MakeVector(Y=0))
  ]}
]
variables []
