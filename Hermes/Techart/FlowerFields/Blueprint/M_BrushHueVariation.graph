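# Translucent unlit brush material with spline mesh usage that adds hue variation to the flower field colors in the RT_FlowerStrips render target.
# Three chained HueShift calls offset the sampled color, each driven by centered greyscale Perlin noise at its own tiling and strength.
# The same noise also feeds desaturation, brightness and opacity branches.
# The graph shows no links to material outputs and holds unconnected Color and overlay blend nodes.
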Material /Hermes/Techart/FlowerFields/Blueprint/M_BrushHueVariation
graph {
 properties {
  bAutomaticallySetUsageInEditor=false
  bUsedWithSplineMeshes=true
  blendMode=BLEND_Translucent
  shadingModel=MSM_Unlit
  usedShadingModels=Unlit
 }
 nodes [
  Constant() {r=1}
  Constant() {r=1}
  n1 = ScalarParameter() {defaultValue=15, parameterName=HueShift3}
  n2 = TextureSample(UVs=Multiply(A=TextureCoordinate(), B=ScalarParameter() {defaultValue=64, parameterName=Tiling3})) {samplerType=SAMPLERTYPE_LinearGrayscale, texture=&T_SimplePerlin_Greyscale}
  n3 = Divide(A=ScalarParameter() {defaultValue=4, parameterName=HueShift2}) {constB=100}
  n4 = TextureSample(UVs=Multiply(A=TextureCoordinate(), B=ScalarParameter() {defaultValue=5, parameterName=Tiling})) {samplerType=SAMPLERTYPE_LinearGrayscale, texture=&T_SimplePerlin_Greyscale}
  n5 = Subtract(A=@n4.R) {constB=0.5}
  n6 = TextureSample(UVs=Multiply(A=TextureCoordinate(), B=ScalarParameter() {defaultValue=16, parameterName=Tiling2})) {samplerType=SAMPLERTYPE_LinearGrayscale, texture=&T_SimplePerlin_Greyscale}
  n7 = Multiply(A=Divide(A=ScalarParameter() {defaultValue=5, parameterName=HueShift1}) {constB=100}, B=Subtract(A=@n6.R) {constB=0.5})
  n8 = TextureSample() {samplerType=SAMPLERTYPE_LinearColor, texture=&/RT_FlowerStrips}
  n9 = MaterialFunctionCall(
   Hue Shift Percentage=Multiply(A=Divide(A=@n1) {constB=100}, B=Subtract(A=@n2.R) {constB=0.5})
   Texture=MaterialFunctionCall(
    Hue Shift Percentage=Multiply(A=@n3, B=@n5)
    Texture=MaterialFunctionCall(Hue Shift Percentage=@n7, Texture=@n8.RGB) {materialFunction=&HueShift}.Result
   ) {materialFunction=&HueShift}.Result
  ) {materialFunction=&HueShift}
  n10 = LinearInterpolate(Alpha=@n6.R) {constA=-50, constB=0}
  Desaturation(None=@n9.Result, Fraction=Multiply(A=ScalarParameter() {defaultValue=1, parameterName=Desat}, B=Saturate(None=@n10)))
  MaterialFunctionCall() {materialFunction=&Blend_Overlay}
  Multiply(A=Saturate(None=LinearInterpolate(Alpha=@n2.R) {constA=-0.5, constB=2}), B=@n9.Result)
  Multiply(A=Min(A=@n6.R, B=@n4.R), B=ScalarParameter() {defaultValue=0.5, parameterName=Opacity})
  OneMinus(None=@n6.R)
  VectorParameter() {parameterName=Color, defaultValue=<rgba 0 0 0 0.5>}
 ]
}
