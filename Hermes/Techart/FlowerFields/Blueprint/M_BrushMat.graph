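# M_BrushMat is an unlit translucent brush material for the FlowerFields tool, enabled for spline meshes, that builds a soft-edged mask in UV space.
# A UV-centered SphereMask with adjustable hardness forms the round shape.
# Distance from center on each axis, inverted, scaled by per-axis falloff and saturated, multiplies into a rectangular shape.
# The SphereRectangle parameter blends circle to rectangle, and the Color parameter's alpha scales the result.
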Material /Hermes/Techart/FlowerFields/Blueprint/M_BrushMat
graph {
 properties {
  bAutomaticallySetUsageInEditor=false
  bUsedWithSplineMeshes=true
  blendMode=BLEND_Translucent
  shadingModel=MSM_Unlit
  usedShadingModels=Unlit
 }
 nodes [
  n1 = TextureCoordinate()
  n2 = Constant2Vector() {g=0.5, r=0.5}
  n3 = ScalarParameter() {defaultValue=0.4439, parameterName=FalloffRadial}
  n4 = Subtract(A=TextureCoordinate(), B=@n2)
  n5 = Divide(A=Abs(None=ComponentMask(None=@n4) {r=true})) {constB=0.5}
  n6 = Saturate(None=LinearInterpolate(B=ScalarParameter() {defaultValue=2.8833, parameterName=FalloffRectX}, Alpha=OneMinus(None=@n5)))
  n7 = Divide(A=Abs(None=ComponentMask(None=@n4) {g=true})) {constB=0.5}
  n8 = Saturate(None=LinearInterpolate(B=ScalarParameter() {defaultValue=4.9986, parameterName=FalloffRectY}, Alpha=OneMinus(None=@n7)))
  Multiply(
   A=LinearInterpolate(
    A=SphereMask(A=@n1, B=@n2, Hardness=@n3) {attenuationRadius=0.5, hardnessPercent=50}
    B=Multiply(A=@n6, B=@n8)
    Alpha=ScalarParameter() {parameterName=SphereRectangle}
   )
   B=VectorParameter() {parameterName=Color, defaultValue=<rgba 1 1 1 1>}.A
  )
 ]
}
