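# Translucent unlit material for spline meshes in the FlowerFields tech art tools, named as a noise overlay for a brush.
# Two tiling samples of a greyscale Perlin texture, centred on zero and scaled by HueShift parameters, drive chained HueShift functions on the FlowerColorProjection texture.
# The minimum of both noises times an Opacity parameter forms a mask.
# The graph also holds a duplicate mask branch and stray Color parameters.
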Material /Hermes/Techart/FlowerFields/Blueprint/M_BrushOverlayNoise
graph {
 properties {
  bAutomaticallySetUsageInEditor=false
  bUsedWithSplineMeshes=true
  blendMode=BLEND_Translucent
  shadingModel=MSM_Unlit
  usedShadingModels=Unlit
 }
 nodes [
  Constant() {r=1}
  n1 = ScalarParameter() {defaultValue=60, parameterName=HueShift2}
  n2 = TextureSample(UVs=Multiply(A=TextureCoordinate(), B=ScalarParameter() {defaultValue=7, parameterName=Tiling})) {samplerType=SAMPLERTYPE_LinearGrayscale, texture=&T_SimplePerlin_Greyscale}
  n3 = Divide(A=ScalarParameter() {defaultValue=50, parameterName=HueShift1}) {constB=100}
  n4 = TextureSample(UVs=Multiply(A=TextureCoordinate(), B=ScalarParameter() {defaultValue=3, parameterName=Tiling2})) {samplerType=SAMPLERTYPE_LinearGrayscale, texture=&T_SimplePerlin_Greyscale}
  n5 = Subtract(A=@n4.R) {constB=0.5}
  MaterialFunctionCall(
   Hue Shift Percentage=Multiply(A=Divide(A=@n1) {constB=100}, B=Subtract(A=@n2.R) {constB=0.5})
   Texture=MaterialFunctionCall(
    Hue Shift Percentage=Multiply(A=@n3, B=@n5)
    Texture=TextureSample() {samplerType=SAMPLERTYPE_LinearColor, texture=&/FlowerColorProjection}.RGB
   ) {materialFunction=&HueShift}.Result
  ) {materialFunction=&HueShift}
  Multiply(A=Min(A=@n4.R, B=@n2.R), B=ScalarParameter() {defaultValue=0.5, parameterName=Opacity})
  n6 = Multiply(A=TextureCoordinate(), B=ScalarParameter() {defaultValue=3, parameterName=Tiling2})
  n7 = TextureSample(UVs=Multiply(A=TextureCoordinate(), B=ScalarParameter() {defaultValue=7, parameterName=Tiling})) {samplerType=SAMPLERTYPE_LinearGrayscale, texture=&T_SimplePerlin_Greyscale}
  Multiply(
   A=Min(
    A=TextureSample(UVs=@n6) {samplerType=SAMPLERTYPE_LinearGrayscale, texture=&T_SimplePerlin_Greyscale}.R
    B=@n7.R
   )
   B=ScalarParameter() {defaultValue=0.5, parameterName=Opacity}
  )
  Multiply(B=@n7.RGB)
  OneMinus(None=@n4.R)
  VectorParameter() {parameterName=Color, defaultValue=<rgba 0 0 0 0.5>}
  VectorParameter() {parameterName=Color, defaultValue=<rgba 0 0 0 0.5>}
 ]
}
