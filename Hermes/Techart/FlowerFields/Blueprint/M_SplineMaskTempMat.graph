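# Unlit spline-mesh material, apparently a temporary mask for FlowerFields, that paints flat red, green and blue bands across the U texture coordinate.
# Two threshold parameters split the bands using Ceil and Saturate as hard steps.
# A single border parameter blacks out both U edges symmetrically.
# The result is multiplied by floored vertex-color red, so only vertices at full red keep color.
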Material /Hermes/Techart/FlowerFields/Blueprint/M_SplineMaskTempMat
graph {
 properties {bAutomaticallySetUsageInEditor=false, bUsedWithSplineMeshes=true, shadingModel=MSM_Unlit, usedShadingModels=Unlit}
 nodes [
  Constant3Vector() {constant=<rgba 1 0.1962 0.246 1>}
  Multiply(A=VertexNormalWS()) {constB=10}
  n1 = ComponentMask(None=TextureCoordinate()) {r=true}
  n2 = Subtract(A=@n1, B=ScalarParameter() {defaultValue=0.373, parameterName=Threshold1})
  n3 = LinearInterpolate(A=Constant3Vector() {constant=<rgba 1 0 0 0>}, B=Constant3Vector() {constant=<rgba 0 1 0 0>}, Alpha=Saturate(None=Ceil(None=@n2)))
  n4 = Constant3Vector() {constant=<rgba 0 0 1 0>}
  n5 = ScalarParameter() {defaultValue=0.629, parameterName=Threshold2}
  n6 = Saturate(None=Ceil(None=Subtract(A=@n1, B=@n5)))
  n7 = ScalarParameter() {defaultValue=0.157, parameterName=ThresholdBlackBorder1}
  n8 = Subtract(A=ComponentMask(None=TextureCoordinate()) {r=true}, B=OneMinus(None=@n7))
  n9 = Saturate(None=OneMinus(None=Ceil(None=@n8)))
  n10 = ComponentMask(None=TextureCoordinate()) {r=true}
  n11 = Saturate(None=Ceil(None=Subtract(A=@n10, B=@n7)))
  Multiply(
   A=LinearInterpolate(A=Constant3Vector(), B=LinearInterpolate(A=@n3, B=@n4, Alpha=@n6), Alpha=Multiply(A=@n9, B=@n11))
   B=Floor(None=VertexColor().R)
  )
 ]
}
