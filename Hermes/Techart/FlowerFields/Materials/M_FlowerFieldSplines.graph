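# Spline mesh material for flower fields in the Hermes terrain, using the Grass physical material.
# World-XY tiled flower textures are tinted by a color projection function, shifting with FOV-compensated distance fade and large-scale noise toward desaturated, contrast-corrected projected color.
# Vertex red drives a height-lerp between the Hermes landscape grass layer and the flowers.
# Snow layers take that blend or its weight, which derives from flower coverage and distance.
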Material /Hermes/Techart/FlowerFields/Materials/M_FlowerFieldSplines
graph {
 properties {
  bAutomaticallySetUsageInEditor=false
  bUseMaterialAttributes=true
  bUsedWithSplineMeshes=true
  physMaterial=/Game/FortressPhysicalMaterials/Grass
  usedShadingModels=Default Lit
 }
 nodes [
  n1 = MaterialFunctionCall!(
   Fade Offset=ScalarParameter() {defaultValue=4000, group=Transition, parameterName=ColorDistanceOffset}
   Fade Distance=ScalarParameter() {defaultValue=6000, group=Transition, parameterName=ColorDistanceFade}
  ) {materialFunction=/Game/Global/Materials/Character_Parent_2023/Anime_Parent_2023/MaterialFunctions/MFC_DistanceFade_FOV_compensation}
  n2 = ScalarParameter() {defaultValue=0.39, group=Transition, parameterName=ColorTransitionMax}
  n3 = Divide(A=WorldPosition().XY, B=ScalarParameter() {defaultValue=1000, parameterName=TilingFlowerFieldFill})
  n4 = TextureSampleParameter2D(UVs=@n3) {group=Textures, parameterName=Base Color, texture=/Hermes/Terrain/Textures/T_Hermes_GrassyFlowers_A_D}
  n5 = Saturate(None=LinearInterpolate(
    A=ScalarParameter() {group=Transition, parameterName=ColorMaskBaseAdd_Low}
    B=ScalarParameter() {defaultValue=1, group=Transition, parameterName=ColorMaskBaseAdd_High}
    Alpha=@n4.A
   ))
  Add(A=Max(A=ScalarParameter() {defaultValue=0.5, group=Transition, parameterName=ColorTransitionmin}, B=Min(A=@n1.Result, B=@n2)), B=@n5)
  Constant() {r=1}
  MaterialFunctionCall!(Grass_UseCustomSlopeMask, Grass_CustomSlopeMask) {materialFunction=/BRCoreMaterials/Hermes/Terrain/Landscape/Materials/Layers/MF_Hermes_StaticMeshLandscapeLayerMatch}
  n6 = MaterialFunctionCall() {materialFunction=/Game/Global/Base4Layers_2022/MaterialFunctions/MF_FlowerFieldsColorProjection}
  n7 = LinearInterpolate(
   A=VectorParameter() {parameterName=ColorFlowerBed, defaultValue=<rgba 1 1 1 1>}.RGB
   B=Multiply(A=@n6.Result, B=ScalarParameter() {defaultValue=1, group=Transition, parameterName=Color intensity})
   Alpha=@n5
  )
  n8 = Multiply(A=@n6.Result, B=ScalarParameter() {defaultValue=1, group=Transition, parameterName=FlowerFieldBrightnessMatch})
  n9 = Multiply(
   A=Power(Base=Desaturation(None=@n4.RGB), Exp=ScalarParameter() {defaultValue=0.568, group=Transition, parameterName=Contrast correction})
   B=ScalarParameter() {defaultValue=0.8, group=Transition, parameterName=Desaturation correction}
  ) {constB=0.85}
  n10 = LinearInterpolate(
   A=ScalarParameter() {group=Transition, parameterName=ColorMaskBaseAddFAR_Low}
   B=ScalarParameter() {defaultValue=1, group=Transition, parameterName=ColorMaskBaseAddFAR_High}
   Alpha=@n4.A
  )
  n11 = Saturate(None=LinearInterpolate(A=Constant(), B=Saturate(None=@n10), Alpha=@n1.Result))
  n12 = TextureSampleParameter2D(UVs=Divide(A=WorldPosition().XY, B=ScalarParameter() {defaultValue=2000, group=Transition, parameterName=LargeScaleNoise})) {parameterName=LargeScaleNoise, texture=&T_Hermes_GrassyFlowers_A_D}
  n13 = Saturate(None=Multiply(A=@n12.A, B=ScalarParameter() {defaultValue=2, group=Transition, parameterName=LargeScaleNoiseMult}))
  n14 = Multiply(A=@n11, B=@n13)
  n15 = Saturate(None=LinearInterpolate(
    A=ScalarParameter() {defaultValue=-5.2154, group=Transition, parameterName=LargeScaleNoise_Low}
    B=ScalarParameter() {defaultValue=2, group=Transition, parameterName=LargeScaleNoise_High}
    Alpha=@n12.A
   ))
  n16 = TextureSampleParameter2D(UVs=@n3) {group=Textures, parameterName=S maps, samplerType=SAMPLERTYPE_Masks, texture=/Hermes/Terrain/Textures/T_Hermes_GrassyFlowers_A_S}
  n17 = TextureSampleParameter2D(UVs=@n3) {group=Textures, parameterName=Normal, samplerType=SAMPLERTYPE_Normal, texture=/Hermes/Terrain/Textures/T_Hermes_GrassyFlowers_A_N}
  n18 = SetMaterialAttributes!(
   MaterialAttributes
   Base Color=LinearInterpolate(A=Multiply(A=@n7, B=@n4.RGB), B=Multiply(A=@n8, B=@n9), Alpha=Multiply(A=@n14, B=@n15))
   Specular=Multiply(A=@n16.R, B=ScalarParameter() {defaultValue=0.4, parameterName=Specular})
   Roughness=Multiply(A=ScalarParameter() {defaultValue=1.1, parameterName=Roughness}, B=@n16.B)
   Normal=@n17.RGB
  ) {attributeSetTypes [
    #1
    #2
    #3
    #4
   ]}
  n19 = MaterialFunctionCall(Transition Phase=Multiply(A=ScalarParameter() {defaultValue=0.65, parameterName=TransitionToGrass}, B=VertexColor().R), Height Texture=@n4.A) {materialFunction=&HeightLerp}
  n20 = BlendMaterialAttributes(
   A=MaterialFunctionCall!(
    Material
    MobileMaterial
    IsForStaticMesh=StaticBool() {value=true}
    UseTangentNormals=StaticBool() {value=true}
    UseCustomSlopeMask
    CustomSlopeMask
    Grass_EdgeOnly
    LayerWeight
    TerrainMasks_QuadA
    TerrainMasks_QuadB
    TerrainMasks_QuadC
    TerrainMasks_QuadD
    VariationMasks Mobile
   ) {materialFunction=/BRCoreMaterials/Hermes/Terrain/Landscape/Materials/Layers/MF_Hermes_Grass}.StaticMeshMaterial
   B=@n18
   Alpha=@n19.Alpha
  )
  n21 = Multiply(A=@n4.A, B=@n19.Alpha)
  n22 = LinearInterpolate(A=Constant() {r=1}, B=Multiply(A=@n15, B=@n13), Alpha=@n11)
  n23 = LinearInterpolate(
   A=ScalarParameter() {defaultValue=1, parameterName=SnowWeight_A}
   B=LinearInterpolate(
    A=ScalarParameter() {defaultValue=-0.1923, parameterName=SnowWeight_B}
    B=ScalarParameter() {defaultValue=-0.6778, parameterName=SnowWeight_B_Far}
    Alpha=@n11
   )
   Alpha=Saturate(None=Multiply(A=@n21, B=@n22))
  )
  MaterialFunctionCall!(
   Material=@n20
   MobileMaterial=@n20
   IsForStaticMesh=StaticBool() {value=true}
   UseTangentNormals=StaticBool() {value=true}
   LayerWeight=@n23
   TerrainMasks_QuadA
   TerrainMasks_QuadB
   TerrainMasks_QuadC
   TerrainMasks_QuadD
   CliffMaterialOnly
   CliffLayerMask
   VariationMasks Mobile
  ) {materialFunction=/BRCoreMaterials/Hermes/Terrain/Landscape/Materials/Layers/MF_Hermes_Snow}
  MaterialFunctionCall(Base=@n9, Blend=@n8) {materialFunction=&Blend_Overlay}
  Multiply(A=MaterialFunctionCall(Transition Phase=@n11, Height Texture=@n4.A) {materialFunction=&HeightLerp}.Alpha)
  n24 = MaterialFunctionCall!(
   ScalarWeightMask=Constant() {r=1}
   UseSnowOcclusionTextureMask=StaticBool()
   ------
   UseMaterialInput
   AdditionalMaskToRemoveSnow
   OutputTangentNormals=StaticBool() {value=true}
   UseAdditionalMaskToRemoveSnow
   UseSnowNormals
   OuputVertexNormalsAsTheSnowNormals
   UseSparkle
   LQSpark
   AllowSnowMicroNormals
   CliffMaterialOnly
   Material
   CliffLandscapeLayerMask
   IsLandscape=StaticBool()
   AreInputNormalsTangentSpace=StaticBool() {value=true}
   UseInputNormalsForFakeSnowShadow
  ) {materialFunction=/Game/Athena/Helios/Environments/Landscape/Materials/QualityLevelLayers/MF_Helios_Snow_Default}
  n25 = MaterialFunctionCall!(
   MobileMaterial=MaterialFunctionCall!(
    InputLayer
    LayerWeight=Constant() {r=1}
    VariationMasks Mobile=MaterialFunctionCall() {materialFunction=/Game/Athena/Helios/Environments/Landscape/Materials/MaterialFunctions_Utilities/MF_VariationMasks_Mobile}.VariationMasks Mobile
   ) {materialFunction=/BRCoreMaterials/Hermes/Terrain/Landscape/Materials/Layers/QualityLevel/MF_Hermes_Snow_Mobile}.Result
   TransformTangents=StaticBool() {value=true}
  ) {materialFunction=/Game/Athena/Helios/Environments/Landscape/Materials/MaterialFunctions_Utilities/MF_MobileDeriveNormals}
  Reroute(None=BlendMaterialAttributes(A=@n20, B=FeatureLevelSwitch(Default=@n24.SnowMaterialOnly, ES3_1=@n25.ResultMaterial), Alpha=Saturate(None=@n23)))
  SetMaterialAttributes!(
   MaterialAttributes
   Emissive Color=MaterialFunctionCall(Number=ScalarParameter() {bUseCustomPrimitiveData=true, parameterName=test}) {materialFunction=&DebugScalarValues}.Result
  ) {attributeSetTypes [#5]}
  n26 = GetMaterialAttributes(None=@n24.SnowMaterialOnly) {attributeGetTypes [#4]}
  SetMaterialAttributes!(MaterialAttributes=@n26.MaterialAttributes, Normal=Normalize(VectorInput=@n26.Normal)) {attributeSetTypes [#4]}
  n27 = WorldPosition()
  n28 = ComponentMask(None=VectorParameter() {parameterName=FlowerField_Origin, defaultValue=<rgba 45088 45472 -4848 1>}.RGB) {g=true, r=true}
  n29 = VectorParameter() {parameterName=FlowerField_Size, defaultValue=<rgba 20000 20000 20000 1>}
  TextureSample(UVs=Divide(A=Subtract(A=@n27.XY, B=@n28), B=ComponentMask(None=@n29.RGB) {g=true, r=true})) {samplerType=SAMPLERTYPE_LinearColor, texture=/Hermes/Techart/FlowerFields/Blueprint/FlowerColorProjection}
 ]
}
